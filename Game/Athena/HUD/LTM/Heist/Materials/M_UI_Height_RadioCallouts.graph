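# Translucent UI material for the Heist LTM radio callout HUD, drawing an animatable bolt shape with a BoltColor parameter.
# The texture's G channel is cut by an anti-aliased threshold at BoltEdge to form a crisp shape.
# Its R channel serves as a wipe gradient: SmoothSteps driven by BoltReveal and inverted BoltOutro multiply into that shape, sharing BoltRevealSoftness for edge width.
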Material /Game/Athena/HUD/LTM/Heist/Materials/M_UI_Height_RadioCallouts
graph {
 properties {blendMode=BLEND_Translucent, materialDomain=MD_UI, pixelDepthOffsetMode=PDOM_Legacy, usedShadingModels=Default Lit}
 nodes [
  n1 = ScalarParameter() {defaultValue=0.5694, parameterName=BoltEdge, sliderMax=1}
  n2 = TextureSample() {texture=&/T_UI_Height_RadioCallouts_CP}
  n3 = ScalarParameter() {parameterName=BoltReveal, sliderMax=1}
  n4 = ScalarParameter() {parameterName=BoltRevealSoftness}
  n5 = Add(A=@n3, B=@n4)
  n6 = ScalarParameter() {defaultValue=1, parameterName=BoltOutro, sliderMax=1}
  n7 = Add(A=@n6, B=@n4)
  Multiply(
   A=Multiply(
    A=MaterialFunctionCall!(Threshold=@n1, Gradient=@n2.G, Filter Width Override, Filter Width Fudge Factor) {materialFunction=/Game/UI/Materials/AAThreshold}.Result
    B=MaterialFunctionCall(Alpha=@n2.R, Min=@n3, Max=@n5) {materialFunction=&SmoothStep}.Result
   )
   B=OneMinus(None=MaterialFunctionCall(Alpha=@n2.R, Min=@n6, Max=@n7) {materialFunction=&SmoothStep}.Result)
  )
  VectorParameter() {parameterName=BoltColor, defaultValue=<rgba 1 1 1 1>}
 ]
}
