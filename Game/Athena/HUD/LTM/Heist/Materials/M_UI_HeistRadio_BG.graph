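# Translucent UI material for the background of the radio image in the Heist LTM HUD.
# A curve atlas gradient sampled along rotated UVs is tinted by the RadioShadow color through the radio texture's alpha, forming a silhouette shadow.
# Shadow scale and offset each use only their red channel for both axes.
# A separate rotated UV feeds a narrow SmoothStep, producing an angled, nearly hard edge mask set by Edge and EdgeAngle.
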Material /Game/Athena/HUD/LTM/Heist/Materials/M_UI_HeistRadio_BG
graph {
 properties {blendMode=BLEND_Translucent, materialDomain=MD_UI, pixelDepthOffsetMode=PDOM_Legacy, usedShadingModels=Default Lit}
 nodes [
  n1 = TextureCoordinate()
  n2 = ScalarParameter() {parameterName=GradientDirection, sliderMax=1}
  n3 = CurveAtlasRowParameter(CurveTime=ComponentMask(None=MaterialFunctionCall(UVs=@n1, "Rotation Angle (0-1)"=@n2) {materialFunction=&CustomRotator}.Rotated Values) {g=true}) {
   atlas=&/HeistGradientAtlas
   curve=&/Heist_RadioBG_GradientCurve
   defaultValue=0.0312
   parameterName=CurveParams
  }
  n4 = VectorParameter() {parameterName=RadioShadow, defaultValue=<rgba 0.2307 0 1 0.23>}
  n5 = TextureCoordinate()
  n6 = VectorParameter() {parameterName=ShadowScale, defaultValue=<rgba 1 1 0 1>}
  n7 = AppendVector(A=@n6.R, B=@n6.R)
  n8 = VectorParameter() {parameterName=ShadowOffset, defaultValue=<rgba 0 0 0 1>}
  n9 = TextureSampleParameter2D(UVs=Add(A=Multiply(A=@n5, B=@n7), B=AppendVector(A=@n8.R, B=@n8.R))) {parameterName=Texture, samplerSource=SSM_Clamp_WorldGroupSettings, texture=&/T_UI_RadioImage}
  LinearInterpolate(A=@n3, B=@n4.RGB, Alpha=Multiply(A=@n9.A, B=@n4.A))
  n10 = TextureCoordinate()
  n11 = ScalarParameter() {parameterName=EdgeAngle}
  n12 = ScalarParameter() {defaultValue=0.5, parameterName=Edge}
  MaterialFunctionCall(
   Alpha=ComponentMask(None=MaterialFunctionCall(UVs=@n10, "Rotation Angle (0-1)"=@n11) {materialFunction=&CustomRotator}.Rotated Values) {g=true}
   Min=@n12
   Max=Subtract(A=@n12) {constB=0.01}
  ) {materialFunction=&SmoothStep}
 ]
}
