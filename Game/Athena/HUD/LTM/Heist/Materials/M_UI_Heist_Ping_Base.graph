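# Translucent base UI material for Heist LTM pings, drawing a scalable, offset icon with animated rings around it.
# Two MF_UI_Ring calls share one looping time but use different easing exponents, so their expanding, softening, fading rings separate.
# Icon alpha blends ring color with icon color, optionally shifted toward a desaturated, tinted scanline hologram.
# Opacity takes the maximum of icon, rings and an optional static-switch outline made by resampling the icon at another scale.
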Material /Game/Athena/HUD/LTM/Heist/Materials/M_UI_Heist_Ping_Base
graph {
 properties {blendMode=BLEND_Translucent, materialDomain=MD_UI, pixelDepthOffsetMode=PDOM_Legacy, usedShadingModels=Default Lit}
 nodes [
  n1 = ScalarParameter() {defaultValue=1, group=Ping, parameterName=PingSpeed}
  n2 = Fmod(A=Multiply(A=Time() {period=1}, B=@n1) {constB=0.5}, B=Constant() {r=1})
  Add(A=Multiply(A=@n2) {constB=0.5}) {constB=0.5}
  n3 = VectorParameter() {parameterName=IconScale, defaultValue=<rgba 0.7048 0.7979 0 1>}
  n4 = MaterialFunctionCall(Texture Scale=AppendVector(A=@n3.R, B=@n3.R)) {materialFunction=&ScaleUVsByCenter}
  n5 = VectorParameter() {parameterName=IconOffset, defaultValue=<rgba 0.0662 0.019 0 1>}
  n6 = AppendVector(A=@n5.R, B=@n5.R)
  n7 = TextureSampleParameter2D(UVs=NamedRerouteDeclaration(None=Add(A=@n4.UVs, B=@n6)) {name=IconUVs, nodeColor=<rgba 0 0 0 1>}) {parameterName=Icon, samplerSource=SSM_Clamp_WorldGroupSettings, texture=&/T_UI_Heist_Diamond_Ally}
  n8 = MaterialFunctionCall!(
   Number of Scanlines=ScalarParameter() {defaultValue=300, group=Hologram, parameterName=Scanline Amount}
   Scanline Speed
   "Scanlines Thickness (Bias)"=ScalarParameter() {defaultValue=-0.15, group=Hologram, parameterName=ScanlineSize}
   Scanlines Dark Value=ScalarParameter() {defaultValue=0.5, group=Hologram, parameterName=Scanlines Dark}
   Scanlines Light Value=ScalarParameter() {defaultValue=1, group=Hologram, parameterName=Scanlines Light}
   Block Opacity
   UVs
   Time
   Block Repeat
  ) {materialFunction=/Game/Effects/Fort_Effects/Effects/Atlas/Materials/MF_Atlas_Scanlines_01}
  n9 = Multiply(A=Desaturation(None=@n7.RGB), B=VectorParameter() {group=Hologram, parameterName=HologramColor, defaultValue=<rgba 0.0999 0.1746 1 1>}.RGB)
  LinearInterpolate(
   A=LinearInterpolate(
    B=VectorParameter() {group=Ping, parameterName=RingColor, defaultValue=<rgba 1 0.7398 0 1>}.RGB
    Alpha=ScalarParameter() {defaultValue=1, group=Ping, parameterName=PingAmount}
   )
   B=LinearInterpolate(A=Add(A=@n7.RGB) {constB=0}, B=Multiply(A=@n8.Result, B=@n9), Alpha=ScalarParameter() {group=Hologram, parameterName=HologramAmount})
   Alpha=@n7.A
  )
  n10 = ScalarParameter() {defaultValue=0.1, group=Ping, parameterName=PingRingMaxSize}
  n11 = Power(Base=@n2) {constExponent=0.4}
  n12 = ScalarParameter() {defaultValue=0.5, group=Ping, parameterName=PingOuterRadius}
  n13 = MaterialFunctionCall!(
   RingSize=LinearInterpolate(A=@n10, Alpha=@n11) {constA=0.1, constB=0.01}
   Edge
   EdgeSoftness=LinearInterpolate(Alpha=@n11) {constA=0.002, constB=0.05}
   OuterRadius=LinearInterpolate(B=@n12, Alpha=@n11) {constB=0.5}
  ) {materialFunction=/Game/UI/InGame/HUD/Textures/MF_UI_Ring}
  n14 = LinearInterpolate(Alpha=@n11) {constA=2, constB=0}
  n15 = Time() {period=1}
  n16 = Power(Base=Fmod(A=Multiply(A=@n15, B=@n1) {constB=0.5}, B=Constant() {r=1})) {constExponent=1.1}
  n17 = MaterialFunctionCall(
   RingSize=LinearInterpolate(A=@n10, Alpha=@n16) {constA=0.1, constB=0.01}
   EdgeSoftness=LinearInterpolate(Alpha=@n16) {constA=0.002, constB=0.05}
   OuterRadius=LinearInterpolate(B=@n12, Alpha=@n16) {constB=0.5}
  ) {materialFunction=&MF_UI_Ring}
  n18 = LinearInterpolate(Alpha=@n16) {constA=2, constB=0}
  n19 = Max(A=Clamp(None=Multiply(A=@n13.Result, B=@n14)), B=Clamp(None=Multiply(A=@n17.Result, B=@n18)))
  n20 = ScalarParameter() {defaultValue=1, group=Ping, parameterName=PingAmount, sliderMax=1}
  n21 = ScalarParameter() {defaultValue=1.08, group=Outline, parameterName=OutlineScale, sliderMax=1}
  n22 = TextureSampleParameter2D(UVs=MaterialFunctionCall(UVs=NamedRerouteUsage(), Texture Scale=AppendVector(A=@n21, B=@n21)) {materialFunction=&ScaleUVsByCenter}.UVs) {parameterName=Icon, samplerSource=SSM_Clamp_WorldGroupSettings, texture=&T_UI_Heist_Diamond_Ally}
  n23 = ScalarParameter() {defaultValue=0.3, group=Outline, parameterName=OutlineOpacity}
  Max(
   A=Max(A=@n7.A, B=Multiply(A=@n19, B=@n20))
   B=StaticSwitchParameter(True=Multiply(A=@n22.A, B=@n23), False=Constant()) {group=OUtline, parameterName=UseOutline}
  )
 ]
}
