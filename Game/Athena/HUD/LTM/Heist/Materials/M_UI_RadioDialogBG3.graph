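# Translucent UI material that draws a radio dialog background for the Heist LTM as dialog, accent and drop shadow layers.
# Texture channels supply dialog (B), accent (R) and bolt (G) gradients; BoltAmount weights the bolt before it Max-merges into each layer.
# AAThreshold cuts antialiased shapes from these gradients, which are layered by lerped colors and Max-combined alpha.
# UV offsets chain from dialog to accent to shadow, using only R of each vector parameter for both axes.
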Material /Game/Athena/HUD/LTM/Heist/Materials/M_UI_RadioDialogBG3
graph {
 properties {blendMode=BLEND_Translucent, materialDomain=MD_UI, pixelDepthOffsetMode=PDOM_Legacy, usedShadingModels=Default Lit}
 nodes [
  n1 = VectorParameter() {group=Shadow, parameterName=ShadowColor, defaultValue=<rgba 0.0052 0.0052 0.0056 0.44>}
  n2 = ScalarParameter() {defaultValue=0.25, parameterName=Edge}
  n3 = VectorParameter() {group=Dialog, parameterName=DialogOffset, defaultValue=<rgba 0 0 0 1>}
  n4 = Add(A=TextureCoordinate(), B=AppendVector(A=@n3.R, B=@n3.R))
  n5 = VectorParameter() {group=Dialog, parameterName=DialogScale, defaultValue=<rgba 1 1 0 1>}
  n6 = Multiply(A=@n4, B=AppendVector(A=@n5.R, B=@n5.R))
  n7 = VectorParameter() {group=Dialog, parameterName=DialogRotationCenter, defaultValue=<rgba 0 0 0 1>}
  n8 = MaterialFunctionCall(
   UVs=@n6
   Rotation Center=AppendVector(A=@n7.R, B=@n7.R)
   "Rotation Angle (0-1)"=ScalarParameter() {group=Dialog, parameterName=DialogRotation}
  ) {materialFunction=&CustomRotator}
  n9 = VectorParameter() {group=Accent, parameterName=AccentOffset, defaultValue=<rgba 0 0 0 1>}
  n10 = Add(A=@n8.Rotated Values, B=AppendVector(A=@n9.R, B=@n9.R))
  n11 = VectorParameter() {group=Bolt, parameterName=BoltShadowScale, defaultValue=<rgba 1 1 0 1>}
  n12 = MaterialFunctionCall!(Texture Scale=AppendVector(A=@n11.R, B=@n11.R), UVs=TextureCoordinate(), Offset XY Texturesize Z) {materialFunction=/Engine/Functions/Engine_MaterialFunctions01/Texturing/ScaleUVsAroundPoint}
  n13 = VectorParameter() {group=Bolt, parameterName=BoltShadowOffset, defaultValue=<rgba 0 0 0 1>}
  n14 = AppendVector(A=@n13.R, B=@n13.R)
  n15 = ScalarParameter() {group=Bolt, parameterName=BoltAmount}
  n16 = Multiply(
   A=TextureSampleParameter2D(UVs=Add(A=@n12.UVs, B=@n14)) {parameterName=Texture, texture=&/T_UI_RadioDialogBG_DF_CP}.G
   B=@n15
  )
  n17 = ScalarParameter() {defaultValue=0.02, parameterName=EdgeSoftness}
  n18 = MaterialFunctionCall!(
   Threshold=@n2
   Gradient=Max(
    A=TextureSampleParameter2D(UVs=@n10) {parameterName=Texture, texture=&T_UI_RadioDialogBG_DF_CP}.R
    B=@n16
   )
   Filter Width Override=@n17
   Filter Width Fudge Factor
  ) {materialFunction=/Game/UI/Materials/AAThreshold}
  n19 = TextureSampleParameter2D() {parameterName=Texture, texture=&T_UI_RadioDialogBG_DF_CP}
  n20 = MaterialFunctionCall(
   Threshold=@n2
   Gradient=Max(
    A=TextureSampleParameter2D(UVs=@n8.Rotated Values) {parameterName=Texture, texture=&T_UI_RadioDialogBG_DF_CP}.B
    B=Multiply(A=@n19.G, B=@n15)
   )
   Filter Width Override=@n17
  ) {materialFunction=&AAThreshold}
  LinearInterpolate(
   A=LinearInterpolate(A=@n1.RGB, B=VectorParameter() {group=Accent, parameterName=AccentColor, defaultValue=<rgba 0 0.9216 1 1>}.RGB, Alpha=@n18.Result)
   B=VectorParameter() {group=Dialog, parameterName=DialogColor, defaultValue=<rgba 1 1 1 1>}.RGB
   Alpha=@n20.Result
  )
  MaterialFunctionCall(UVs=@n4, Texture Scale=@n6) {materialFunction=&ScaleUVsByCenter}
  n21 = VectorParameter() {group=Shadow, parameterName=ShadowOffset, defaultValue=<rgba 0 0 0 1>}
  n22 = Add(A=@n10, B=AppendVector(A=@n21.R, B=@n21.R))
  n23 = MaterialFunctionCall(
   Threshold=@n2
   Gradient=Max(
    A=TextureSampleParameter2D(UVs=@n22) {parameterName=Texture, texture=&T_UI_RadioDialogBG_DF_CP}.R
    B=@n16
   )
   Filter Width Override=@n17
  ) {materialFunction=&AAThreshold}
  Max(A=Max(A=Multiply(A=@n23.Result, B=@n1.A), B=@n18.Result), B=@n20.Result)
 ]
}
