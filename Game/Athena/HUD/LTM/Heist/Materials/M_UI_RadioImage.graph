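# Translucent UI material for a radio image in LTM HUDs, distorted by a fake-depth bump and switchable to Burger or Tomato team art.
# A rotated gradient times an exponential radial falloff, scaled and added to a speaker displacement texture, forms a height field.
# BumpOffset_Advanced uses that height and a CameraVector parameter, rather than the real view, to warp texture UVs.
# TeamActive and TeamType select the artwork, while RadiusSwap blends the base image toward the height field.
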Material /Game/Athena/HUD/LTM/Heist/Materials/M_UI_RadioImage
graph {
 properties {blendMode=BLEND_Translucent, materialDomain=MD_UI, pixelDepthOffsetMode=PDOM_Legacy, usedShadingModels=Default Lit}
 nodes [
  n1 = VectorParameter() {parameterName=HeightBase, defaultValue=<rgba 0 0 0 1>}
  n2 = Add(A=TextureCoordinate(), B=ScalarParameter() {parameterName=GradientOffset})
  n3 = ScalarParameter() {parameterName=GradientRotation}
  n4 = VectorParameter() {parameterName=RadialScale, defaultValue=<rgba 1 1 0 1>}
  n5 = Multiply(A=TextureCoordinate(), B=AppendVector(A=@n4.R, B=@n4.R))
  n6 = TextureCoordinate()
  n7 = ScalarParameter() {parameterName=UVTiltDirection}
  n8 = LinearInterpolate(
   A=ComponentMask(None=MaterialFunctionCall(UVs=@n6, "Rotation Angle (0-1)"=@n7) {materialFunction=&CustomRotator}.Rotated Values) {g=true}
   Alpha=ScalarParameter() {defaultValue=1, parameterName=UVTilt}
  )
  n9 = VectorParameter() {parameterName=BumpCenter, defaultValue=<rgba 0.5 0.5 0 1>}
  n10 = Multiply(
   A=ComponentMask(None=MaterialFunctionCall(UVs=@n2, "Rotation Angle (0-1)"=@n3) {materialFunction=&CustomRotator}.Rotated Values) {g=true}
   B=MaterialFunctionCall(
    UVs=Multiply(A=@n5, B=@n8)
    CenterPosition=AppendVector(A=@n9.R, B=@n9.R)
    Radius=ScalarParameter() {defaultValue=0.5, parameterName=BumpRadius}
    Density=ScalarParameter() {defaultValue=2, parameterName=BumpDensity}
    Invert Density=StaticBool() {value=true}
   ) {materialFunction=&RadialGradientExponential}.RadialGradientExponential
  )
  n11 = Multiply(
   A=MaterialFunctionCall(Float3=Add(A=@n1.RGB, B=@n10)) {materialFunction=&BreakOutFloat3Components}.R
   B=ScalarParameter() {defaultValue=0.5, parameterName=FullBodyDistort}
  )
  n12 = TextureSampleParameter2D() {parameterName=SpeakerDisplace, texture=&/T_UI_RadioImage_B}
  n13 = Add(A=@n11, B=@n12.R)
  n14 = MaterialFunctionCall!(
   HeightTexture=@n13
   TextureCoordinates=TextureCoordinate()
   ReferencePlane=ScalarParameter() {parameterName=RefPlane}
   HeightRatioInput=ScalarParameter() {defaultValue=0.5, parameterName=HeightRatio, sliderMax=1}
   CameraVector=VectorParameter() {parameterName=CameraVector, defaultValue=<rgba 0 0 0 1>}.RGB
  ) {materialFunction=/Engine/Functions/Engine_MaterialFunctions02/Utility/BumpOffset_Advanced.BumpOffset_advanced}
  n15 = TextureSampleParameter2D(UVs=@n14.UVs) {parameterName=Texture, samplerSource=SSM_Clamp_WorldGroupSettings, texture=&/T_UI_RadioImage}
  LinearInterpolate(
   A=LinearInterpolate(A=@n15.RGB, B=@n13, Alpha=ScalarParameter() {parameterName=RadiusSwap, sliderMax=1})
   B=LinearInterpolate(
    A=TextureSampleParameter2D(UVs=@n14.UVs) {parameterName=TextureBurger, texture=/Game/Athena/HUD/LTM/Barrier/Materials/T_UI_RadioBurger}.RGB
    B=TextureSampleParameter2D(UVs=@n14.UVs) {parameterName=TextureTomato, texture=/Game/Athena/HUD/LTM/Barrier/Materials/T_UI_RadioTomato}.RGB
    Alpha=ScalarParameter() {defaultValue=1, parameterName=TeamType, sliderMax=1}
   )
   Alpha=ScalarParameter() {parameterName=TeamActive, sliderMax=1}
  )
  Max(A=@n11, B=@n12.R)
  Reroute(None=@n15.A)
 ]
}
